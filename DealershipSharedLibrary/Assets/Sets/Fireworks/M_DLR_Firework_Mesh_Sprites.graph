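# Translucent unlit material for firework sparks on Niagara mesh particles, drawing mesh quads as flare sprites.
# World position offset combines a Sprite function, an offset along vertex normals and a gravity drop weighted by vertex red, driven by dynamic parameters.
# Vertex green varies size, offset, gravity, rotation, hue shift, noise flicker and dissolve per spark.
# Particle alpha blends flicker toward steady emissive; opacity combines flare mask, dissolve and optional depth fade.
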Material /DealershipSharedLibrary/Assets/Sets/Fireworks/M_DLR_Firework_Mesh_Sprites
graph {
 properties {
  bCastRayTracedShadows=false
  bUsedWithInstancedStaticMeshes=true
  bUsedWithNanite=true
  bUsedWithNiagaraMeshParticles=true
  bUsedWithSplineMeshes=true
  blendMode=BLEND_Translucent
  pixelDepthOffsetMode=PDOM_Legacy
  shadingModel=MSM_Unlit
  twoSided=true
  usedShadingModels=Unlit
 }
 nodes [
  n1 = VertexNormalWS()
  n2 = VertexColor()
  n3 = Multiply(
   A=LinearInterpolate(Alpha=DynamicParameter() {defaultValue=<rgba 1 1 -1 1>, paramNames [NoiseLerp,Gravity,OffsetMesh,LineWidth]}.OffsetMesh) {constA=-1}
   B=Multiply(A=LinearInterpolate(Alpha=@n2.G) {constA=0.8, constB=1.2}, B=ScalarParameter() {defaultValue=100})
  )
  n4 = WorldPosition()
  n5 = Multiply(
   A=DynamicParameter() {parameterIndex=3, defaultValue=<rgba 1 1 -1 1>, paramNames [SpriteSize,Param2,Param3,Param4]}.SpriteSize
   B=LinearInterpolate(Alpha=VertexColor().G) {constA=0.8, constB=1.2}
  )
  n6 = Multiply(
   A=DynamicParameter() {defaultValue=<rgba 1 1 -1 1>, paramNames [NoiseLerp,Gravity,OffsetMesh,LineWidth]}.Gravity
   B=LinearInterpolate(Alpha=VertexColor().G) {constA=0.8, constB=1.2}
  ) {constB=-100}
  n7 = VertexColor()
  n8 = Multiply(A=@n7.R, B=@n7.R)
  Add(
   A=Add(
    A=Multiply(A=@n1, B=@n3)
    B=MaterialFunctionCall!(Center Location=@n4.XYZ, 0-1 UVs, XY scale=@n5, OverridePivot, "Normals (Optional)") {materialFunction=&Sprite}.World position offset
   )
   B=Multiply(
    A=MaterialFunctionCall(Z=@n6) {materialFunction=&MakeFloat3}.Result
    B=Multiply(A=@n8, B=@n8)
   )
  )
  n9 = Add(A=VertexColor().G) {constB=-0.5}
  n10 = DynamicParameter() {parameterIndex=1, defaultValue=<rgba 0 0 1 1>, paramNames [Dissolve,HueShift,Param3,Param4]}
  n11 = MaterialFunctionCall(UVs=TextureCoordinate(), "Rotation Angle (0-1)"=VertexColor().G) {materialFunction=&CustomRotator}
  n12 = TextureSampleParameter2D(UVs=@n11.Rotated Values) {parameterName=FlareMask, texture=/Game/Effects/Fort_Effects/Textures/Flares/T_Flare_Spikeball}
  n13 = Multiply(A=ParticleColor().RGB, B=Add(A=@n12.R) {constB=0.5})
  n14 = Constant() {r=0.1}
  n15 = VertexColor()
  n16 = TextureSample(UVs=Panner(Coordinate=Add(A=@n14, B=@n15.G)) {speedX=0.3, speedY=0.5}) {samplerType=SAMPLERTYPE_LinearColor, texture=/Game/Global/Textures/Default/Noise/Noise32x32}
  n17 = Multiply(A=@n16.R, B=@n16.R)
  n18 = ParticleColor()
  n19 = LinearInterpolate(
   A=Multiply(A=Multiply(A=@n17, B=@n17)) {constB=8}
   B=ScalarParameter() {defaultValue=5, parameterName=Emissive}
   Alpha=Multiply(A=@n18.A, B=@n18.A)
  )
  MaterialFunctionCall(Emissive=MaterialFunctionCall(Hue Shift Percentage=Multiply(A=@n9, B=@n10.HueShift), Texture=Multiply(A=@n13, B=@n19)) {materialFunction=&HueShift}.Result) {materialFunction=&MF_FX_EmissiveAdjustments}
  n20 = MaterialFunctionCall(Float2=Multiply(A=@n11.Rotated Values, B=OneMinus(None=@n11.Rotated Values))) {materialFunction=&BreakOutFloat2Components}
  n21 = Multiply(A=@n20.R, B=@n20.R)
  n22 = Subtract(A=VertexColor().G)
  n23 = Multiply(A=OneMinus(None=DynamicParameter() {parameterIndex=1, defaultValue=<rgba 0 0 1 1>, paramNames [Dissolve,HueShift,Param3,Param4]}.Dissolve)) {constB=2}
  n24 = Multiply(A=Multiply(A=@n12.G, B=Ceil(None=@n21)), B=Saturate(None=Add(A=@n22, B=@n23)))
  n25 = ShadingPathSwitch(Default=@n24, Deferred=DepthFade(Opacity=@n24, FadeDistance=ScalarParameter() {defaultValue=100, parameterName=FadeDistance})) {desc="Defered is pc, Forward is switch, mobile is mobile"}
  n26 = FeatureLevelSwitch(Default=@n24, ES2=@n24, ES3_1=@n24, SM4_REMOVED=@n24, SM5=@n25, SM6=@n25)
  n27 = Constant() {r=1}
  Multiply(
   A=Multiply(A=@n24, B=StaticSwitchParameter(True=@n26, False=@n27) {group=Opacity, parameterName=UseDephtFade, sortPriority=10})
   B=VertexColor().A
  )
 ]
}
