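# Unlit, two-sided masked material for firework sparks on spline-mesh Niagara particles, driven by dynamic parameters.
# World position offset stretches vertices along tangents, thickens lines with depth- and screen-height-aware SplineThicken, and adds gravity droop weighted by vertex red to the fourth power.
# Particle color gets vertex-green hue shift and HDR boost for emissive.
# Opacity mask combines a capsule texture, two panning noises, vertex-green-varied erosion and particle alpha.
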Material /DealershipSharedLibrary/Assets/Sets/Fireworks/M_DLR_Firework_Sparks_Masked
graph {
 properties {
  bEnableResponsiveAA=true
  bUseTranslucencyVertexFog=false
  bUsedWithMeshParticles=true
  bUsedWithNanite=true
  bUsedWithNiagaraMeshParticles=true
  blendMode=BLEND_Masked
  pixelDepthOffsetMode=PDOM_Legacy
  previewMesh=&/SM_SplineMesh_Fireworks
  shadingModel=MSM_Unlit
  twoSided=true
  usedShadingModels=Unlit
 }
 nodes [
  n1 = Multiply(A=VertexTangentWS(), B=VertexColor().R)
  n2 = VertexColor()
  n3 = Multiply(
   A=LinearInterpolate(Alpha=DynamicParameter() {defaultValue=<rgba 1 1 -1 1>, paramNames [NoiseLerp,Gravity,OffsetMesh,LineWidth]}.OffsetMesh) {constA=-1}
   B=Multiply(A=LinearInterpolate(Alpha=@n2.G) {constA=0.8, constB=1.2}, B=ScalarParameter() {defaultValue=150, parameterName=Param_1})
  )
  n4 = DynamicParameter() {defaultValue=<rgba 1 1 -1 1>, paramNames [NoiseLerp,Gravity,OffsetMesh,LineWidth]}
  n5 = PixelDepth()
  n6 = ScalarParameter() {defaultValue=0.1, desc=Multiplier for adding spline width based on size on screen, group=SplineWidth, parameterName=ScreenWidthMult}
  n7 = Multiply(A=ComponentMask(None=ViewSize()) {g=true}, B=@n6)
  n8 = ScalarParameter() {defaultValue=5, desc=Max clamp on how much screen size can add to spline width, group=SplineWidth, parameterName=ScreenWidthMaxAdd}
  n9 = Add(A=Clamp(None=Divide(A=@n5.R, B=@n7), Max=@n8))
  n10 = Multiply(A=@n4.LineWidth, B=@n9)
  n11 = Multiply(A=@n4.LineWidth, B=@n9)
  n12 = StaticBool()
  n13 = Multiply(
   A=DynamicParameter() {defaultValue=<rgba 1 1 -1 1>, paramNames [NoiseLerp,Gravity,OffsetMesh,LineWidth]}.Gravity
   B=LinearInterpolate(Alpha=VertexColor().G) {constA=0.8, constB=1.2}
  ) {constB=-100}
  n14 = VertexColor()
  n15 = Multiply(A=@n14.R, B=@n14.R)
  Add(
   A=Add(
    A=Multiply(A=@n1, B=@n3)
    B=MaterialFunctionCall(WidthBase=@n10, WidthTip=@n11, Expand U or V UV Channel=@n12) {materialFunction=&SplineThicken}.WorldPositionOffset
   )
   B=Multiply(
    A=MaterialFunctionCall(Z=@n13) {materialFunction=&MakeFloat3}.Result
    B=Multiply(A=@n15, B=@n15)
   )
  )
  DynamicParameter() {parameterIndex=3, defaultValue=<rgba 1 1 -1 1>, paramNames [SpriteSize,Param2,Param3,Param4]}
  n16 = VertexColor()
  n17 = TextureCoordinate()
  n18 = MaterialFunctionCall(
   Hue Shift Percentage=Multiply(
    A=Add(A=@n16.G) {constB=-0.5}
    B=DynamicParameter() {parameterIndex=1, defaultValue=<rgba 0 0 1 1>, paramNames [Erosion,HueShift,Param3,Param4]}.HueShift
   )
   Texture=Multiply(A=ComponentMask(None=@n17) {g=true}, B=ParticleColor().RGB)
  ) {materialFunction=&HueShift}
  MaterialFunctionCall(Emissive=MaterialFunctionCall(LDR Input=Saturate(None=@n18.Result), HDR Tint=Multiply(A=@n18.Result) {constB=5}) {materialFunction=&DeriveHDRfromLDR}.RemappedRange) {materialFunction=&MF_FX_EmissiveAdjustments}
  n19 = TextureCoordinate()
  n20 = Constant2Vector() {g=1, r=0.5}
  n21 = TextureSample(UVs=Multiply(A=Add(A=@n19, B=@n20), B=Constant2Vector() {g=1, r=0.49})) {samplerSource=SSM_Wrap_WorldGroupSettings, texture=/Game/Missions/Secondary/Events/HonkingFirework/FX/Textures/T_FireWorks_Capsule_Sparks_01}
  n22 = VectorParameter() {parameterName=BaseNoise1, channelNames {a=Speed Y, b=Speed X, g=Scale Y, r=Scale X}, defaultValue=<rgba 3.14 0.048 0.1 0>}
  n23 = Multiply(A=Add(A=TextureCoordinate(), B=VertexColor().G), B=AppendVector(A=@n22.Scale X, B=@n22.Scale X))
  n24 = AppendVector(A=@n22.Speed X, B=@n22.Speed X)
  n25 = VectorParameter() {parameterName=BaseNoise2, channelNames {a=Speed Y, b=Speed X, g=Scale Y, r=Scale X}, defaultValue=<rgba 1 0.082 -0.03 0>}
  n26 = Multiply(A=Add(A=TextureCoordinate(), B=VertexColor().G), B=AppendVector(A=@n25.Scale X, B=@n25.Scale X))
  n27 = AppendVector(A=@n25.Speed X, B=@n25.Speed X)
  n28 = Multiply(
   A=TextureSample(UVs=Panner(Coordinate=@n23, Speed=@n24) {speedX=-0.04}) {texture=/Game/Global/Textures/Default/Noise/T_TileNoise_02}.G
   B=TextureSample(UVs=Panner(Coordinate=@n26, Speed=@n27) {speedX=-0.04}) {texture=&T_TileNoise_02}.G
  )
  n29 = Constant() {r=200}
  n30 = LinearInterpolate(
   A=@n21.G
   B=Multiply(A=Multiply(A=@n28, B=@n29), B=@n21.G)
   Alpha=DynamicParameter() {defaultValue=<rgba 1 1 -1 1>, paramNames [NoiseLerp,Gravity,OffsetMesh,LineWidth]}.NoiseLerp
  )
  n31 = ComponentMask(None=TextureCoordinate()) {r=true}
  n32 = VertexColor()
  n33 = DynamicParameter() {parameterIndex=1, defaultValue=<rgba 0 0 1 1>, paramNames [Erosion,HueShift,Param3,Param4]}
  n34 = Add(A=Subtract(A=Multiply(A=@n32.G, B=@n32.G)), B=Multiply(A=OneMinus(None=@n33.Erosion)) {constB=2})
  Multiply(A=Multiply(A=Multiply(A=@n30, B=@n31), B=Saturate(None=@n34)), B=Multiply(A=VertexColor().R, B=ParticleColor().A))
 ]
}
